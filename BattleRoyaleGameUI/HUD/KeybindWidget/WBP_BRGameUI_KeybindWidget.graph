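# Keybind widget for the Battle Royale game UI HUD, a thin Blueprint subclass of the native FortKeybindWidget.
# Its only designer widget is a root Image_Preview showing a PS5 face-down button icon, apparently as a preview placeholder.
# That image defaults to Collapsed and Construct collapses it again; the Blueprint adds no variables or other logic of its own.
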WidgetBlueprint /BattleRoyaleGameUI/HUD/KeybindWidget/WBP_BRGameUI_KeybindWidget
graphs [EventGraph {nodes [
   Event Construct()
   > SetVisibility(self=Get Image_Preview() -> (Image_Preview:Image Object Reference), InVisibility=Collapsed)
  ]}]
variables []
widgets {
 info {inheritedWidgetCount=0, parentClass=/Script/FortniteUI.FortKeybindWidget, rootWidgetClass=&Image, widgetCount=1}
 widgets [Image_Preview {
   widgetClassPath=&Image
   bIsVariable=true
   widget=Image_Preview
   properties {
    visibility=Collapsed
    brush {
     bIsDynamicallyLoaded=false
     drawAs=Image
     imageType=FullColor
     mirroring=NoMirror
     resourceName=None
     resourceObject=/FortUILibrary/Input/GamepadPS5/MaterialInstances/MI_UI_Input_PS5_Face_Down
     tiling=NoTile
     imageSize=<32 32>
     margin=<ltrb 0 0 0 0>
     outlineSettings {
      bUseBrushTransparency=false
      roundingType=HalfHeightRadius
      width=0
      color {colorUseRule=UseColor_Specified, specifiedColor=<rgba 0 0 0 0>}
      cornerRadii=<0 0 0 0>
     }
     tintColor {colorUseRule=UseColor_Specified, specifiedColor=<rgba 1 1 1 1>}
     uVRegion {bIsValid=false, max=<0 0>, min=<0 0>}
    }
   }
  }]
}
